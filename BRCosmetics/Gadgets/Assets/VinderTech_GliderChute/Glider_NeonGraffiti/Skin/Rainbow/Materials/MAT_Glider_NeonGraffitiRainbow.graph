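# Surface material for the Rainbow skin of the Neon Graffiti glider, a skeletal-mesh cosmetic.
# Shared weapon functions build the base attributes, skipping the elemental effect on mobile, then add skin FX and Creative move-tool support.
# A stepped rainbow function from the matching character skin restyles the result.
# The SkinFX_Mask green channel then blends in outlines with black base color and tinted emissive, before accessory tech-art overrides.
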Material /BRCosmetics/Gadgets/Assets/VinderTech_GliderChute/Glider_NeonGraffiti/Skin/Rainbow/Materials/MAT_Glider_NeonGraffitiRainbow
graph {
 properties {
  bUseMaterialAttributes=true
  bUsedWithMeshDeformer=true
  bUsedWithMorphTargets=true
  bUsedWithSkeletalMesh=true
  numCustomizedUVs=5
  pixelDepthOffsetMode=PDOM_Legacy
  usedShadingModels=Default Lit
 }
 nodes [
  n1 = MaterialFunctionCall!(UV 1 coordinates, MuzzleFalloff) {materialFunction=/Game/Packages/Fortnite_UniqueMaterials_1/Weapons/M_GunMasterShaderFunction}
  n2 = MaterialFunctionCall!(In=@n1.Material Result) {materialFunction=/Game/Weapons/WeaponSkins/MaterialFunction/WeaponEmissiveTOD}
  n3 = MaterialFunctionCall() {materialFunction=/Game/Packages/Fortress_SharedMaterials/Base_Material_Functions/BaseCustomizationFunctions/MF_Base_LocalPosition}
  n4 = MaterialFunctionCall!(Material Attributes=ShadingPathSwitch(
    Default=MaterialFunctionCall!(
     In=@n2.Result
     Pre-Skinned Local Position=@n1.Pre-Skinned Local Position Pixel Shader
     Pre-Skinned Local Normal=@n3.Pre-Skinned Local Normal Pixel Shader
    ) {materialFunction=/Game/Weapons/WeaponSkins/MaterialFunction/ElementalWeaponsFunction}.Result
    Mobile=@n2.Result
   ) {desc=Migrated from FeatureLevelSwitch}) {materialFunction=/Game/Packages/Fortnite_UniqueMaterials_1/Weapons/ADSScopeMaster}
  n5 = MaterialFunctionCall!(In=MaterialFunctionCall(MA In=MaterialFunctionCall!(Pre-Skinned World Normal, Gmap, Customization Mask, IsCharacter, IsStatsB, Material Attributes=@n4.Result) {materialFunction=/Game/Weapons/WeaponSkins/MaterialFunction/Weapon_SkinFXFunction}.Result) {materialFunction=&MF_CreativeMoveTool}.MA Out) {materialFunction=/BRCosmetics/Characters/Player/Female/Medium/Bodies/F_MED_Neon_Graffiti/Skin/Rainbow/Materials/MF_Stepped_GraffitiRainbow}
  n6 = GetMaterialAttributes(None=@n5.Result) {attributeGetTypes [#1, #2]}
  n7 = Multiply(
   A=VectorParameter() {group=NG_Rainbow, parameterName=NG_OutlineColor, defaultValue=<rgba 0.1202 0 1 1>}.RGB
   B=ScalarParameter() {defaultValue=1, group=NG_Rainbow, parameterName=NG_OutlineMult}
  )
  n8 = TextureSampleParameter2D() {group=NG_Rainbow, parameterName=SkinFX_Mask, samplerType=SAMPLERTYPE_Masks, texture=&T_BlackMask}
  MaterialFunctionCall(In=SetMaterialAttributes!(
    MaterialAttributes=@n6.MaterialAttributes
    Emissive Color=LinearInterpolate(A=@n6.Emissive Color, B=@n7, Alpha=@n8.G)
    Base Color=LinearInterpolate(A=@n6.Base Color, Alpha=@n8.G) {constB=0}
   ) {attributeSetTypes [#1, #2]}) {materialFunction=&MF_TechArtOverrides_accessories}
 ]
}
